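# Sound cue for the impact of the "Bucket Nice" item, routed through the Fortnite weapons sound class.
# A mixer layers a pitch- and volume-modulated base impact wave at full level with a half-level random pick from four numbered impact variations, which has its own modulator.
# An external surface attenuation asset shapes falloff, while a concurrency override caps simultaneous instances and virtualization is disabled.
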SoundCue /FNGameplayCues/Athena/Sounds/Items/Bucket/Bucket_Nice_Impact_Cue
root=Mixer_0
nodes [
 {id=Mixer_0, children [Modulator_1,Modulator_0], properties {inputVolume [1,0.5]}}
 {id=Modulator_0, children [Random_0]}
 {id=Modulator_1, children [WavePlayer_0], properties {pitchMax=1.2, volumeMax=1, volumeMin=0.8}}
 {
  id=Random_0
  children [WavePlayer_1,WavePlayer_2,WavePlayer_3,WavePlayer_4]
  properties {bShouldExcludeFromBranchCulling=true, weights [1,1,1,1]}
 }
 {id=WavePlayer_0, properties {soundWaveAssetPtr=&/Bucket_Nice_Impact}}
 {id=WavePlayer_1, properties {soundWaveAssetPtr=&/Bucket_Nice_Impact_01}}
 {id=WavePlayer_2, properties {soundWaveAssetPtr=&/Bucket_Nice_Impact_02}}
 {id=WavePlayer_3, properties {soundWaveAssetPtr=&/Bucket_Nice_Impact_03}}
 {id=WavePlayer_4, properties {soundWaveAssetPtr=&/Bucket_Nice_Impact_04}}
]
properties {
 attenuationSettings=&/Bucket_Nice_Impact_Surface_Att
 bOverrideConcurrency=true
 maxDistance=18300
 soundClassObject=/Game/Sounds/FortSoundClassesAndModes/Fort_Weapons
 virtualizationMode=Disabled
 volumeMultiplier=0.7
 attenuationOverrides {distanceAlgorithm=Logarithmic, falloffDistance=3000, attenuationShapeExtents {x=2500}}
 concurrencyOverrides {maxCount=3}
}
