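# Sound cue for the Bucket item's impact sound in Athena, filed among the gameplay cue sounds.
# A single wave player feeds the Bucket_Timpact wave straight to the output, with no randomization or modulation nodes.
# It routes to the Fort_Impacts sound class and borrows the shared medium explosion attenuation for distance falloff.
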SoundCue /FNGameplayCues/Athena/Sounds/Items/Bucket/Bucket_Timpact_Cue
root=WavePlayer_0
nodes [{id=WavePlayer_0, properties {soundWaveAssetPtr=&/Bucket_Timpact}}]
properties {
 attenuationSettings=/Game/Sounds/Attenuation/Fort_Explosion_Medium_Athena_Att
 maxDistance=5087.5
 soundClassObject=&Fort_Impacts
 volumeMultiplier=1
}
